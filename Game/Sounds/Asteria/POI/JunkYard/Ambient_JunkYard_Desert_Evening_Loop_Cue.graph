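# Looping outdoor ambience cue for an evening desert junkyard point of interest.
# The first two quality slots mix randomly chosen industrial beds and quad desert night wind with randomly chosen distant metal rattling.
# The last two use a lighter, randomly modulated pair: stereo desert night and one fixed industrial bed.
# Endlessly looping envelopes, nested at different lengths in the richer branch, slowly vary the sound to mask repetition.
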SoundCue /Game/Sounds/Asteria/POI/JunkYard/Ambient_JunkYard_Desert_Evening_Loop_Cue
root=QualityLevel_0
nodes [
 {
  id=Enveloper_0
  children [Mixer_1]
  properties {
   bLoop=true
   bLoopIndefinitely=true
   loopEnd=30
   pitchCurve {editorCurveData {keys [{interpMode=RCIM_Cubic, time=0}, <at 15.9468 0.9181 RCIM_Cubic>, {interpMode=RCIM_Cubic, time=30}]}}
   volumeCurve {editorCurveData {keys [<at 0 1>]}}
  }
 }
 {
  id=Enveloper_1
  children [Mixer_0]
  properties {
   bLoop=true
   bLoopIndefinitely=true
   loopEnd=30
   pitchCurve {editorCurveData {keys [{interpMode=RCIM_Cubic, time=0}, <at 15.9468 0.9181 RCIM_Cubic>, {interpMode=RCIM_Cubic, time=30}]}}
   volumeCurve {editorCurveData {keys [<at 0 1 RCIM_Cubic>,<at 16.1421 0.921 RCIM_Cubic>,<at 30 1 RCIM_Cubic>]}}
  }
 }
 {
  id=Enveloper_2
  children [Random_1]
  properties {
   bLoop=true
   bLoopIndefinitely=true
   loopEnd=30
   pitchCurve {editorCurveData {keys [{interpMode=RCIM_Cubic, time=0}, <at 15.9468 0.9181 RCIM_Cubic>, {interpMode=RCIM_Cubic, time=30}]}}
   volumeCurve {editorCurveData {keys [<at 0 1>]}}
  }
 }
 {
  id=Enveloper_4
  children [Enveloper_2]
  properties {
   bLoop=true
   bLoopIndefinitely=true
   loopEnd=40
   pitchCurve {editorCurveData {keys [{time=0}]}}
   volumeCurve {editorCurveData {keys [
      {arriveTangent=-0.01, interpMode=RCIM_Cubic, time=0, value=1}
      <at 10 0.6 RCIM_Cubic>
      {arriveTangent=-0.0105, interpMode=RCIM_Cubic, leaveTangent=-0.0105, tangentMode=RCTM_User, time=20, value=1}
      {arriveTangent=-0.025, interpMode=RCIM_Cubic, leaveTangent=-0.025, time=23.5506, value=0.4}
      {arriveTangent=0.0013, interpMode=RCIM_Cubic, leaveTangent=0.0013, tangentMode=RCTM_User, time=28, value=0.8}
      {arriveTangent=0.0167, interpMode=RCIM_Cubic, leaveTangent=0.0167, time=33, value=0.6}
      <at 40 1>
     ]}}
  }
 }
 {
  id=Enveloper_6
  children [Enveloper_0]
  properties {
   bLoop=true
   bLoopIndefinitely=true
   loopEnd=100
   pitchCurve {editorCurveData {keys [<at 0 0.8>]}}
   volumeCurve {editorCurveData {keys [<at 0 1 RCIM_Cubic>,<at 50 0.9 RCIM_Cubic>,<at 100 1 RCIM_Cubic>]}}
  }
 }
 {id=Mixer_0, children [WavePlayer_2,WavePlayer_0], properties {inputVolume [0.5,1]}}
 {id=Mixer_1, children [Modulator_11,WavePlayer_1], properties {inputVolume [1,0.5]}}
 {id=Mixer_2, children [Enveloper_6,Enveloper_4], properties {inputVolume [1,0.6]}}
 {id=Modulator_11, children [Random_0], properties {pitchMax=1, pitchMin=0.97}}
 {id=Modulator_2, children [Enveloper_1], properties {pitchMax=1.1, pitchMin=0.9, volumeMax=1.1, volumeMin=0.9}}
 {id=QualityLevel_0, children [Mixer_2,Mixer_2,Modulator_2,Modulator_2]}
 {id=Random_0, children [WavePlayer_8, WavePlayer_9, WavePlayer_10, WavePlayer_11], properties {weights [1,1,1,1]}}
 {id=Random_1, children [WavePlayer_3,WavePlayer_6], properties {weights [1,1]}}
 {id=WavePlayer_0, properties {bLooping=true, soundWaveAssetPtr=/Game/Sounds/Asteria/POI/IndustrialTown/Ambience_Industrial_Evening_03_Quad}}
 {id=WavePlayer_1, properties {bLooping=true, soundWaveAssetPtr=/Game/Sounds/Artemis/Biomes/Desert/Ambience_Desert_Night_Quad}}
 {id=WavePlayer_10, properties {bLooping=true, soundWaveAssetPtr=&Ambience_Industrial_Evening_03_Quad}}
 {id=WavePlayer_11, properties {bLooping=true, soundWaveAssetPtr=/Game/Sounds/Asteria/POI/IndustrialTown/Ambience_Industrial_Evening_04_Quad}}
 {id=WavePlayer_2, properties {bLooping=true, soundWaveAssetPtr=/Game/Sounds/Artemis/Biomes/Desert/Ambience_Desert_Night_Stereo}}
 {
  id=WavePlayer_3
  properties {bLooping=true, soundWaveAssetPtr=&/Ambience_Metal_Distant_Rattling_In_Wind_Loop_03_Quad}
 }
 {
  id=WavePlayer_6
  properties {bLooping=true, soundWaveAssetPtr=&/Ambience_Metal_Distant_Rattling_In_Wind_Loop_04_Quad}
 }
 {id=WavePlayer_8, properties {bLooping=true, soundWaveAssetPtr=/Game/Sounds/Asteria/POI/IndustrialTown/Ambience_Industrial_Evening_01_Quad}}
 {id=WavePlayer_9, properties {bLooping=true, soundWaveAssetPtr=/Game/Sounds/Asteria/POI/IndustrialTown/Ambience_Industrial_Evening_02_Quad}}
]
properties {maxDistance=2097152, soundClassObject=&Fort_Ambient_Outdoors_Athena, volumeMultiplier=0.501}
